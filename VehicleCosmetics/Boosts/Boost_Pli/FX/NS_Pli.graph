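# Niagara vehicle boost cosmetic that trails green vines and scattered leaves from filtered skeletal-mesh sockets.
# Vine spawns ribbons per port, with four renderers and four sine/cosine-offset positions forming twisting strands that taper and fade.
# Leaf spawns shrinking sprites with curl noise and per-update random sub-images.
# Both inherit fading source movement and vary motion by game-mode mask and frontend-preview flag; painted-variant overrides are disabled.
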NiagaraSystem /VehicleCosmetics/Boosts/Boost_Pli/FX/NS_Pli
dependencies {
 usedDataInterfaces [&NiagaraDataInterfaceCurve, &NiagaraDataInterfaceSkeletalMesh, &NiagaraDataInterfaceArrayFloat3]
 usedDynamicInputs [
  &MergeHelper_Vector2Switch
  &SpawnGroup
  &MergeHelper_FloatSwitch
  &SelectVectorFromArray
 ]
 usedModules [
  &self:SystemSpawnScript.SystemScriptSource.SystemScriptGraph.NiagaraNodeAssignment_3.SetVariables_#1
  &self:SystemSpawnScript.SystemScriptSource.SystemScriptGraph.NiagaraNodeAssignment_4.SetVariables_#2
  &self:SystemSpawnScript.SystemScriptSource.SystemScriptGraph.NiagaraNodeAssignment_2.SetVariables_#3
  &SystemState
 ]
}
emitters {
 NE_VC_Boost_Leaf {
  emitterUpdateScript [
   &EmitterState()
   &VC_FilteredSocketCount()
   &SpawnRate(
    SpawnRate=&Multiply_Float(
     A=300
     B=/VehicleCosmetics/FX/Niagara/Modules/VC_Mask_Floats(Coro Value:NiagaraFloat=1, FrontEnd Value:NiagaraFloat=1, FN Value:NiagaraFloat=0.5)
    )
    Use Spawn Probability=false
   )
  ]
  particleSpawnScript [
   &VC_PaintedVariantsOverride() {enabled=false}
   &InitializeParticle(
    Color=&LerpLinearColors(
     StartColor=<rgba 0.12 0.3 0 1>
     EndColor=<rgba 0.06 0.2 0 1>
     LerpFactor=&UniformAOrBFloat(A=0, Distribution Weight=0.4, Evaluation Type=::NewEnumerator0, Recalculate Random Each Loop=true, Override Seed=false)
    )
    Lifetime=&RandomRangeFloat(Minimum=0.4, Maximum=0.6)
    Sprite Rotation Mode=::NewEnumerator2
    Sprite Size Mode=::NewEnumerator3
    Sprite UV Mode=::NewEnumerator3
    Uniform Sprite Size=&MakeCustomFloatFromBool(
     Bool=$User.bIsFrontendPreview
     True Float=&RandomRangeFloat(Minimum=25, Maximum=45)
     False Float=&RandomRangeFloat(Minimum=40, Maximum=60)
    )
    UsePositionOffset=false
    Write Material Random=true
   )
   &PartitionParticles(Number Of Partitions=$Emitter.PortCount)
   &SkeletalMeshLocation(
    Bone / Socket Index=$Particles.PartitionParticles.Partition
    Bone / Socket Sampling Mode=::NewEnumerator5
    MaskSpawnGroup=false
    Mesh Sampling Type=::NewEnumerator6
    Sampled Position Offset=&MakeCustomVectorFromBool(
     Bool=$User.bIsFrontendPreview
     True Vector=&RandomRangeVector(Minimum=<-30 -30 -30>, Maximum=<30 30 30>)
     False Vector=&RandomRangeVector(Minimum=<-20 -20 -20>, Maximum=<20 20 20>)
    )
    Skeletal Mesh=$System.SkeletalMesh
   )
   &AddVelocity(Velocity=&VC_Mask_Vectors(
     Coro Value:Vector3f=<-200 0 0>
     FrontEnd Value:Vector3f=&MakeCustomVectorFromBool(
      Bool=$User.bIsFrontendPreview
      True Vector=&RandomRangeVector(Minimum=<-2000 -150 -150>, Maximum=<-4000 150 150>)
      False Vector=&RandomRangeVector(Minimum=<0 -150 -150>, Maximum=<-600 150 150>)
     )
     FN Value:Vector3f=<-1200 0 0>
    ))
   /VehicleCosmetics/FX/VC_FXParents/Modules/DLR_ProductPreviews(GameModeID:NiagaraInt32=$System.GameModeID, bPreviewVehicleSpeed:NiagaraBool=$System.bPreviewVehicleSpeed)
  ]
  particleUpdateScript [
   &ParticleState()
   &CurlNoiseForce(Bias Noise=false, Noise Frequency=20, Noise Strength=3000, Pan Noise=false)
   &ScaleSpriteSize(
    Scale Sprite Size Mode=::NewEnumerator0
    Uniform Scale Factor=&FloatFromCurve(FloatCurve=DataInterface() {propertyValues {Curve {keys [<at 0.5 1 RCIM_Cubic>,<at 1 0 RCIM_Cubic>]}}})
   )
   &InheritSourceMovement(
    Applied Position Scale=&VectorFromFloat(Value=&FloatFromCurve(FloatCurve=DataInterface() {propertyValues {Curve {
         defaultValue=1
         keys [
          {arriveTangent=-1.7885, interpMode=RCIM_Cubic, leaveTangent=-1.7885, tangentMode=RCTM_User, time=0.05, value=1}
          {interpMode=RCIM_Cubic, tangentMode=RCTM_User, time=1, value=0}
         ]
        }}}))
    ApplyForce=false
    ApplyPosition=true
    ApplyPositionFalloff=false
   )
   &ScaleVelocity(
    Coordinate Space:ENiagaraCoordinateSpace=::Simulation
    Velocity Scale=&VectorFromFloat(Value=&FloatFromCurve(FloatCurve=DataInterface() {propertyValues {Curve {keys [{interpMode=RCIM_Cubic, leaveTangentWeight=0.1685, tangentMode=RCTM_User, time=0, value=1}, <at 0.5 0.5 RCIM_Cubic>]}}}))
   )
   &SolveForcesAndVelocity(Clamp Velocity=false, Limit Acceleration=false)
   &DynamicMaterialParameters(
    Index 0 Param 1=1
    Index 0 Param 2=1
    Index 1 Param 2=0.4
    Module.Index0Float0Write=true
    Module.Index0Float1Write=true
    Module.Index0Float2Write=true
    Module.Index0Float3Write=true
    Module.Index1Float0Write=true
    Module.Index1Float1Write=true
    Module.Index1Float2Write=true
    Module.Index1Float3Write=true
    Write Parameter Index 1=::NewEnumerator1
   )
   /SetVariables(Particles.SubImageIndex=&RandomRangeFloat(Minimum=0, Maximum=15)) {isSetParameters=true}
  ]
  rendererClasses [&NiagaraSpriteRendererProperties]
  renderers [{rendererClass=&NiagaraSpriteRendererProperties}]
 }
 NE_VC_Boost_Vine {
  emitterUpdateScript [
   &EmitterState()
   &VC_FilteredSocketCount()
   &VC_RateQualityHandler(Default Rate:NiagaraFloat=&Multiply_Float(
     A=100
     B=&VC_Mask_Floats(Coro Value:NiagaraFloat=1.2, FrontEnd Value:NiagaraFloat=1, FN Value:NiagaraFloat=0.5)
    ))
   &VC_SpawnRibbon_v1(Number of Ribbons=$Emitter.PortCount, Spawn Time=-1, SpawnRate=$Emitter.Spawn Rate)
   /SetVariables(Emitter.RibbonWidth:Vector4f=/Niagara/DynamicInputs/TypeConversions/MakeCustomVector4FromBool(Bool:NiagaraBool=true, True Vector4:Vector4f=<40 50 50 40>, False Vector4:Vector4f=<50 60 60 50>)) {isSetParameters=true}
  ]
  particleSpawnScript [
   &VC_PaintedVariantsOverride() {enabled=false}
   &VC_InitializeRibbon_v1(
    Color=<rgba 0.1 0.25 0 1>
    Enable Skeletal Position=true
    Lifetime=0.8
    Offset Amount=5
    Ribbon Width:NiagaraFloat=&MakeFloatFromVector4(Vector 4=$Emitter.RibbonWidth, Channel=::NewEnumerator0)
   )
   /SetVariables(Particles.RibbonWidth01:NiagaraFloat=&MakeFloatFromVector4(Vector 4=$Emitter.RibbonWidth, Channel=::NewEnumerator1)) {isSetParameters=true}
   /SetVariables(Particles.RibbonWidth02:NiagaraFloat=&MakeFloatFromVector4(Vector 4=$Emitter.RibbonWidth, Channel=::NewEnumerator2)) {isSetParameters=true}
   /SetVariables(Particles.RibbonWidth03:NiagaraFloat=&MakeFloatFromVector4(Vector 4=$Emitter.RibbonWidth, Channel=::NewEnumerator3)) {isSetParameters=true}
   &AddVelocity(Velocity=&VC_Mask_Vectors(
     Coro Value:Vector3f=<-400 0 0>
     FrontEnd Value:Vector3f=&MakeCustomVectorFromBool(Bool=$User.bIsFrontendPreview, True Vector=<-2400 0 1>, False Vector=<-600 0 0>)
     FN Value:Vector3f=<-1200 0 0>
    )) {module=AddVelocity001}
   &DLR_ProductPreviews(GameModeID:NiagaraInt32=$System.GameModeID, bPreviewVehicleSpeed:NiagaraBool=$System.bPreviewVehicleSpeed)
  ]
  particleUpdateScript [
   &ParticleState()
   /SetVariables(Particles.Position01:NiagaraPosition=&AddVectorToPosition(
     Position=$Particles.Position
     Vector=&TransformVector(
      Vector=&MakeVector(
       X=0
       Y=&Cosine(
        Normalized Angle=&Subtract_Float(A=$Emitter.Age, B=$Particles.NormalizedAge)
        Period=&MakeCustomFloatFromBool(Bool=$User.bIsFrontendPreview, True Float=0.35, False Float=0.6)
        Scale=&FloatFromCurve(
         FloatCurve=DataInterface() {propertyValues {Curve {keys [{arriveTangent=15.9799, interpMode=RCIM_Cubic, leaveTangent=15.9799, tangentMode=RCTM_User, time=0.05, value=0}, <at 0.2 1>]}}}
         Scale Curve=10
        )
        Bias=0
       )
       Z=&Sine(
        Normalized Angle=&Subtract_Float(A=$Emitter.Age, B=$Particles.NormalizedAge)
        Period=&MakeCustomFloatFromBool(Bool=$User.bIsFrontendPreview, True Float=0.3, False Float=0.4)
        Scale=&FloatFromCurve(
         FloatCurve=DataInterface() {propertyValues {Curve {keys [{arriveTangent=15.9799, interpMode=RCIM_Cubic, leaveTangent=15.9799, tangentMode=RCTM_User, time=0.05, value=0}, <at 0.2 1>]}}}
         Scale Curve=-10
        )
        Bias=0
       )
      )
      Destination Space=::World
     )
    )) {isSetParameters=true}
   /SetVariables(Particles.Position02:NiagaraPosition=&AddVectorToPosition(
     Position=$Particles.Position
     Vector=&TransformVector(
      Vector=&MakeVector(
       X=0
       Y=&Sine(
        Normalized Angle=&Subtract_Float(A=$Emitter.Age, B=$Particles.NormalizedAge)
        Period=&MakeCustomFloatFromBool(Bool=$User.bIsFrontendPreview, True Float=0.4, False Float=0.4)
        Scale=&FloatFromCurve(
         FloatCurve=DataInterface() {propertyValues {Curve {keys [{arriveTangent=15.9799, interpMode=RCIM_Cubic, leaveTangent=15.9799, tangentMode=RCTM_User, time=0.05, value=0}, <at 0.2 1>]}}}
         Scale Curve=-10
        )
        Bias=5
       )
       Z=&Cosine(
        Normalized Angle=&Subtract_Float(A=$Emitter.Age, B=$Particles.NormalizedAge)
        Period=&MakeCustomFloatFromBool(Bool=$User.bIsFrontendPreview, True Float=0.35, False Float=0.6)
        Scale=&FloatFromCurve(
         FloatCurve=DataInterface() {propertyValues {Curve {keys [{arriveTangent=15.9799, interpMode=RCIM_Cubic, leaveTangent=15.9799, tangentMode=RCTM_User, time=0.05, value=0}, <at 0.2 1>]}}}
         Scale Curve=10
        )
        Bias=5
       )
      )
      Destination Space=::World
     )
    )) {isSetParameters=true}
   /SetVariables(Particles.Position03:NiagaraPosition=&AddVectorToPosition(
     Position=$Particles.Position
     Vector=&TransformVector(
      Vector=&MakeVector(
       X=0
       Y=&Sine(
        Normalized Angle=&Subtract_Float(A=$Emitter.Age, B=$Particles.NormalizedAge)
        Period=&MakeCustomFloatFromBool(Bool=$User.bIsFrontendPreview, True Float=0.4, False Float=0.6)
        Scale=&FloatFromCurve(
         FloatCurve=DataInterface() {propertyValues {Curve {keys [{arriveTangent=15.9799, interpMode=RCIM_Cubic, leaveTangent=15.9799, tangentMode=RCTM_User, time=0.05, value=0}, <at 0.2 1>]}}}
         Scale Curve=10
        )
        Bias=5
       )
       Z=&Cosine(
        Normalized Angle=&Subtract_Float(A=$Emitter.Age, B=$Particles.NormalizedAge)
        Period=&MakeCustomFloatFromBool(Bool=$User.bIsFrontendPreview, True Float=0.3, False Float=0.4)
        Scale=&FloatFromCurve(
         FloatCurve=DataInterface() {propertyValues {Curve {keys [{arriveTangent=15.9799, interpMode=RCIM_Cubic, leaveTangent=15.9799, tangentMode=RCTM_User, time=0.05, value=0}, <at 0.2 1>]}}}
         Scale Curve=-10
        )
        Bias=5
       )
      )
      Destination Space=::World
     )
    )) {isSetParameters=true}
   /SetVariables(Particles.Position04:NiagaraPosition=&AddVectorToPosition(
     Position=$Particles.Position
     Vector=&TransformVector(
      Vector=&MakeVector(
       X=0
       Y=&Cosine(
        Normalized Angle=&Subtract_Float(A=$Emitter.Age, B=$Particles.NormalizedAge)
        Period=&MakeCustomFloatFromBool(Bool=$User.bIsFrontendPreview, True Float=0.3, False Float=0.6)
        Scale=&FloatFromCurve(
         FloatCurve=DataInterface() {propertyValues {Curve {keys [{arriveTangent=15.9799, interpMode=RCIM_Cubic, leaveTangent=15.9799, tangentMode=RCTM_User, time=0.05, value=0}, <at 0.2 1>]}}}
         Scale Curve=-10
        )
        Bias=0
       )
       Z=&Sine(
        Normalized Angle=&Subtract_Float(A=$Emitter.Age, B=$Particles.NormalizedAge)
        Period=&MakeCustomFloatFromBool(Bool=$User.bIsFrontendPreview, True Float=0.35, False Float=0.4)
        Scale=&FloatFromCurve(
         FloatCurve=DataInterface() {propertyValues {Curve {keys [{arriveTangent=15.9799, interpMode=RCIM_Cubic, leaveTangent=15.9799, tangentMode=RCTM_User, time=0.05, value=0}, <at 0.2 1>]}}}
         Scale Curve=10
        )
        Bias=0
       )
      )
      Destination Space=::World
     )
    )) {isSetParameters=true}
   &InheritSourceMovement(
    Applied Position Scale=&VectorFromFloat(Value=&FloatFromCurve(FloatCurve=DataInterface() {propertyValues {Curve {
         defaultValue=1
         keys [
          {arriveTangent=-1.7885, interpMode=RCIM_Cubic, leaveTangent=-1.7885, tangentMode=RCTM_User, time=0.05, value=1}
          {interpMode=RCIM_Cubic, tangentMode=RCTM_User, time=1, value=0}
         ]
        }}}))
    ApplyForce=false
    ApplyPosition=true
    ApplyPositionFalloff=false
   )
   &SolveForcesAndVelocity(Clamp Velocity=false, Limit Acceleration=false)
   &VC_UpdateRibbon_v1(Enable Skeletal Position=true)
   &ScaleRibbonWidth(Ribbon Width Scale=&FloatFromCurve(FloatCurve=DataInterface() {propertyValues {Curve {keys [
         {arriveTangent=-0.1437, interpMode=RCIM_Cubic, leaveTangent=-0.1437, tangentMode=RCTM_User, time=0, value=1}
         {arriveTangent=-1.3699, interpMode=RCIM_Cubic, leaveTangent=-1.3699, tangentMode=RCTM_User, time=1, value=0.2}
        ]}}}))
   &DynamicMaterialParameters(
    Index 0 Param 1=1
    Module.Index0Float0Write=true
    Module.Index0Float1Write=true
    Module.Index0Float2Write=true
    Module.Index0Float3Write=true
   )
   &ScaleColor(
    Scale Alpha=&FloatFromCurve(FloatCurve=DataInterface() {propertyValues {Curve {keys [
         <at 0 0 RCIM_Cubic>
         {arriveTangentWeight=0.0581, interpMode=RCIM_Cubic, leaveTangentWeight=0.3488, tangentMode=RCTM_User, time=0.15, value=1}
         {interpMode=RCIM_Cubic, leaveTangent=-4.7443, tangentMode=RCTM_Break, time=0.7, value=1}
         <at 1 0 RCIM_Cubic>
        ]}}})
    ScaleA=true
    ScaleRGB=true
   )
  ]
  rendererClasses [&NiagaraRibbonRendererProperties]
  renderers [
   {rendererClass=&NiagaraRibbonRendererProperties}
   {rendererClass=&NiagaraRibbonRendererProperties, rendererIndex=1}
   {rendererClass=&NiagaraRibbonRendererProperties, rendererIndex=2}
   {rendererClass=&NiagaraRibbonRendererProperties, rendererIndex=3}
  ]
 }
}
summary {
 systemName=NS_Pli
 emitters [
  {emitterName=NE_VC_Boost_Vine, rendererClasses [&NiagaraRibbonRendererProperties]}
  {emitterName=NE_VC_Boost_Leaf, rendererClasses [&NiagaraSpriteRendererProperties]}
 ]
 userVariables [
  User.bIsFrontendPreview {type=NiagaraBool}
  User.bIsPainted {type=NiagaraBool}
  User.blsIsCoroColor {type=NiagaraBool}
  User.bPreviewVehicleSpeed {type=NiagaraBool}
  User.ColorVariantID {type=NiagaraInt32}
  User.GameModeID {type=NiagaraInt32}
  User.PaintedColor {type=LinearColor, description="", defaultValue {struct=LinearColor, value=<rgba 1 1 1 1>}}
  User.PaintedScalar {type=NiagaraFloat, defaultValue=1, description=""}
  User.PortCount {type=NiagaraInt32, defaultValue=2}
 ]
}
